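# EUBP_BiomeBrush_Path is an actor-based Editor Utility Blueprint in the BiomeBrushManager tech-art folder, named as a path biome brush.
# BeginPlay, ActorBeginOverlap, Tick and the Construction Script only forward to their parent implementations.
# It has no variables or logic of its own here, so any brush behavior would come from its parent class, which the file does not name.
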EditorUtilityBlueprint /Game/NeverCook/ArtTools/Tech_Art/BiomeBrushManager/EUBP_BiomeBrush_Path
graphs [
 EventGraph {nodes [
   Event BeginPlay()
   > Parent: BeginPlay() {is=title K2Node_CallParentFunction}
   n1 = Event ActorBeginOverlap()
   > Parent: ActorBeginOverlap!(execute, OtherActor:Actor Object Reference=@n1.OtherActor) -> (then) {is=title K2Node_CallParentFunction}
   n2 = Event Tick!() -> (OutputDelegate, then, DeltaSeconds:float)
   > Parent: Tick!(execute, DeltaSeconds:float=@n2.DeltaSeconds) -> (then) {is=title K2Node_CallParentFunction}
  ]}
 UserConstructionScript {nodes [
   Construction Script()
   > Parent: Construction Script() {is=title K2Node_CallParentFunction}
  ]}
]
variables []
