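# Translucent unlit material for biome brushes, drawn without depth testing, with color and page index supplied per primitive through custom primitive data.
# A FlipBook picks the PageIndex digit from a numbers texture and tiles it in screen-aligned pixel space.
# In perspective views the digit darkens the color when ShowPageIndex is set.
# IsOrthographic switches orthographic views to the plain color and drives a second blend from half to full value, likely opacity.
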Material /Game/NeverCook/ArtTools/Tech_Art/BiomeBrushManager/M_BiomeBrush
graph {
 properties {
  bAutomaticallySetUsageInEditor=false
  bDisableDepthTest=true
  bFullyRough=true
  bUsedWithInstancedStaticMeshes=true
  bUsedWithNanite=true
  blendMode=BLEND_Translucent
  ditherOpacityMask=true
  shadingModel=MSM_Unlit
  usedShadingModels=Unlit
 }
 nodes [
  n1 = VectorParameter() {bUseCustomPrimitiveData=true, parameterName=Color, primitiveDataIndex=1}
  n2 = Constant() {r=11}
  n3 = Divide(A=ScalarParameter() {bUseCustomPrimitiveData=true, parameterName=PageIndex}, B=@n2) {constB=10}
  n4 = Constant() {r=1}
  n5 = TextureObject() {samplerType=SAMPLERTYPE_Masks, texture=/Game/Athena/Playlists/Barrier/Textures/T_Numbers2_M}
  n6 = Constant() {r=256}
  n7 = Frac(None=Multiply(A=MaterialFunctionCall!(TextureResolution=@n6) {materialFunction=/Engine/Functions/Engine_MaterialFunctions02/Texturing/ScreenAlignedPixelToPixelUVs}.Result) {constB=12})
  n8 = Multiply(A=ComponentMask(None=MaterialFunctionCall("Animation  Phase (0-1)"=@n3, Number of Rows=@n4, Number of Columns=@n2, Texture=@n5, UVs=@n7) {materialFunction=&FlipBook}.Result) {r=true}) {constB=0.33}
  n9 = ScalarParameter() {bUseCustomPrimitiveData=true, parameterName=ShowPageIndex, primitiveDataIndex=5}
  LinearInterpolate(A=Subtract(A=@n1.RGB, B=LinearInterpolate(B=@n8, Alpha=@n9)), B=@n1.RGB, Alpha=IsOrthographic())
  LinearInterpolate(Alpha=IsOrthographic()) {constA=0.5}
  ScalarParameter() {parameterName=twgw}
 ]
}
