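; PCG graph for an area biome brush that fills its spline interior with cell-sized debug cubes using the biome brush material.
; Cell size comes from the BiomeBrushManager, setting scale-compensated sample spacing and cube scale, with positions rounded and duplicates pruned.
; Before sampling, it ranks all brushes by height and writes this actor's index to SortPrio, using Gather to enforce order.
; An extents-and-projection branch feeds nothing downstream.
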(graph PCG_BiomeBrush_Area
  (bind GetSplineData_0 (Get%20Spline%20Data :ActorSelector (struct :ActorSelection ByClass :bSelectMultiple true) :bTrackActorsOnlyWithinBounds false))
  (bind GetActorData_57 (Get%20Actor%20Data :Mode GetSinglePoint))
  (bind GetAttributeFromPointIndex_46 (Get%20Attribute%20From%20Point%20Index :bForceOutputAttributeToBeInElementsDomain true :InputSource $Scale :OutputAttributeName Scale))
  (bind BreakVectorAttribute_47 (Break%20Vector%20Attribute))
  (bind GetActorProperty_0 (Get%20Actor%20Property :ActorSelector (struct :ActorFilter AllWorldActors :ActorSelection ByClass :ActorSelectionClass /Game/NeverCook/ArtTools/Tech_Art/BiomeBrushManager/EUBP_BiomeBrushManager.EUBP_BiomeBrushManager_C) :bProcessAllComponents false :bSanitizeOutputAttributeName false :OutputAttributeName None :PropertyName BiomeCellSize))
  (bind AttributeMathsOp_51 (Attribute%20Maths%20Op :Operation Divide))
  (bind GetActorData_42 (Get%20Actor%20Data :Mode GetSinglePoint))
  (bind GetActorData_39 (Get%20Actor%20Data :ActorSelector (struct :ActorFilter AllWorldActors :ActorSelection ByClass :ActorSelectionClass /Game/NeverCook/ArtTools/Tech_Art/BiomeBrushManager/EUBP_BiomeBrush_Base.EUBP_BiomeBrush_Base_C :bSelectMultiple true) :bMergeSinglePointData true :bTrackActorsOnlyWithinBounds false :Mode GetSinglePoint))
  (bind SortAttributes_40 (Sort%20Attributes :bUseStableSort false :SortAttributes [(struct :InputSource $Position.Z)]))
  (bind CopyAttribute_48 (Copy%20Attributes :InputSource $Index :OutputTarget SortPrio))
  (bind CreateAttribute_46 (Create%20Constant :AttributeTypes (struct :DoubleValue 1.0)))
  (bind GetPointsCount_44 (Get%20Element%20Count))
  (bind AttributeMathsOp_45 (Attribute%20Maths%20Op :Operation Subtract))
  (bind DuplicatePoint_43 (Duplicate%20Point))
  (bind AttributeFilter_41 (Filter%20Attribute%20Elements :Operator Matches :TargetAttribute ActorReference :ThresholdAttribute ActorReference))
  (bind GetAttributeFromPointIndex_49 (Get%20Attribute%20From%20Point%20Index :bForceOutputAttributeToBeInElementsDomain true :OutputAttributeName SortPrio))
  (bind ApplyOnActor_51 (Apply%20On%20Object :ObjectReferenceAttribute ActorReference :PropertyOverrideDescriptions [(struct :InputSource SortPrio :PropertyTarget "SortPrio")]))
  (bind Gather_53 (Gather))
  (bind SplineSampler_1 (Spline%20Sampler :SamplerParams (struct :Dimension OnInterior :DistanceIncrement 5e+03 :InteriorSampleSpacing 1e+03 :InteriorBorderSampleSpacing 1e+03 :bUnbounded true)))
  (bind CreateAttribute_33 (Create%20Constant :AttributeTypes (struct :DoubleValue 1e+02)))
  (bind AttributeMathsOp_44 (Attribute%20Maths%20Op :Operation Divide))
  (bind CreateAttribute_42 (Create%20Constant :AttributeTypes (struct :DoubleValue 1.0)))
  (bind MakeVectorAttribute_41 (Make%20Vector%20Attribute :OutputType Vector))
  (bind AttributeMathsOp_38 (Attribute%20Maths%20Op :InputSource1 $Scale :Operation Set))
  (bind AttributeMathsOp_36 (Attribute%20Maths%20Op :InputSource1 $Position :Operation Round))
  (bind SelfPruning_38 (Self%20Pruning :Parameters (struct :PruningType RemoveDuplicates :ComparisonSource $Extents :bRandomizedPruning false)))
  (bind StaticMeshSpawner_2 (Static%20Mesh%20Spawner :bApplyMeshBoundsToPoints false :MeshSelectorParameters (class:/Script/PCG.PCGMeshSelectorWeighted :MeshEntries [(struct :Descriptor (struct :StaticMesh /PCG/DebugObjects/PCG_Cube.PCG_Cube :OverrideMaterials [/Game/NeverCook/ArtTools/Tech_Art/BiomeBrushManager/M_BiomeBrush.M_BiomeBrush] :Mobility Movable :bHasCustomNavigableGeometry No :BodyInstance (struct :VelocitySolverIterationCount 1 :CollisionEnabled NoCollision :CollisionProfileName NoCollision :CollisionResponses (struct :ResponseArray [(struct :Channel Visibility :Response ECR_Ignore) (struct :Channel Camera :Response ECR_Ignore)]) :bAutoWeld true) :InstanceStartCullDistance 99999999 :InstanceEndCullDistance 999999999 :bCastShadow false :bAffectDynamicIndirectLighting false :bAffectDistanceFieldLighting false :bCastDynamicShadow false :bCastStaticShadow false :bCastContactShadow false :bReceivesDecals false :bUseGpuLodSelection false :bIncludeInHLOD false))]) :PostProcessFunctionNames [PostSpawn]))
  (bind GetActorData_24 (Get%20Actor%20Data :ActorSelector (struct :ActorFilter AllWorldActors :ActorSelection ByClass) :Mode GetSinglePoint))
  (bind CreateAttribute_28 (Create%20Constant :AttributeTypes (struct :DoubleValue 0.1)))
  (bind ExtentsModifier_27 (Extents%20Modifier :Extents (0.15 0.15 0.15) :Mode Multiply))
  (bind Projection_25 (Projection :ProjectionParams (struct :bProjectPositions false :bProjectRotations false)))
  (connect GetActorData_42 Out ApplyOnActor_51 In)
  (connect GetAttributeFromPointIndex_49 Attribute ApplyOnActor_51 Property%20Overrides)
  (connect DuplicatePoint_43 Out AttributeFilter_41 Filter)
  (connect CopyAttribute_48 Out AttributeFilter_41 In)
  (connect AttributeMathsOp_38 Out AttributeMathsOp_36 In)
  (connect SplineSampler_1 Out AttributeMathsOp_38 InA)
  (connect MakeVectorAttribute_41 Out AttributeMathsOp_38 InB)
  (connect GetActorProperty_0 Out AttributeMathsOp_44 InA)
  (connect CreateAttribute_33 Out AttributeMathsOp_44 InB)
  (connect GetPointsCount_44 Out AttributeMathsOp_45 InA)
  (connect CreateAttribute_46 Out AttributeMathsOp_45 InB)
  (connect GetActorProperty_0 Out AttributeMathsOp_51 InA)
  (connect BreakVectorAttribute_47 X AttributeMathsOp_51 InB)
  (connect GetAttributeFromPointIndex_46 Attribute BreakVectorAttribute_47 In)
  (connect SortAttributes_40 Out CopyAttribute_48 Source)
  (connect SortAttributes_40 Out CopyAttribute_48 Target)
  (connect GetActorData_42 Out DuplicatePoint_43 In)
  (connect AttributeMathsOp_45 Out DuplicatePoint_43 Iterations)
  (connect CreateAttribute_28 Out ExtentsModifier_27 Extents)
  (connect AttributeMathsOp_38 Out ExtentsModifier_27 In)
  (connect ApplyOnActor_51 Out Gather_53 Execution%20Dependency)
  (connect GetSplineData_0 Out Gather_53 In)
  (connect GetActorData_57 Out GetAttributeFromPointIndex_46 In)
  (connect AttributeFilter_41 InsideFilter GetAttributeFromPointIndex_49 In)
  (connect GetActorData_39 Out GetPointsCount_44 In)
  (connect AttributeMathsOp_44 Out MakeVectorAttribute_41 X)
  (connect AttributeMathsOp_44 Out MakeVectorAttribute_41 Y)
  (connect CreateAttribute_42 Out MakeVectorAttribute_41 Z)
  (connect ExtentsModifier_27 Out Projection_25 In)
  (connect GetActorData_24 Out Projection_25 Projection%20Target)
  (connect AttributeMathsOp_36 Out SelfPruning_38 In)
  (connect GetActorData_39 Out SortAttributes_40 In)
  (connect AttributeMathsOp_51 Out SplineSampler_1 InteriorSampleSpacing)
  (connect Gather_53 Out SplineSampler_1 Spline)
  (connect SelfPruning_38 Out StaticMeshSpawner_2 In))
