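; PCG graph for a path biome brush that sets the brush actor's sort priority and draws it as cubes along its spline.
; It sorts all biome brush actors by height, matches this actor by reference, and writes its index to SortPrio.
; A Gather holds spline sampling until that write finishes.
; Evenly spaced samples, scaled in XY from their bounds, spawn non-colliding, shadowless M_BiomeBrush cubes, then PostSpawn runs.
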(graph PCG_BiomeBrush_Path
  (bind GetSplineData_0 (Get%20Spline%20Data :ActorSelector (struct :ActorSelection ByClass :bSelectMultiple true) :bTrackActorsOnlyWithinBounds false))
  (bind GetActorData_3 (Get%20Actor%20Data :Mode GetSinglePoint))
  (bind GetActorData_0 (Get%20Actor%20Data :ActorSelector (struct :ActorFilter AllWorldActors :ActorSelection ByClass :ActorSelectionClass /Game/NeverCook/ArtTools/Tech_Art/BiomeBrushManager/EUBP_BiomeBrush_Base.EUBP_BiomeBrush_Base_C :bSelectMultiple true) :bMergeSinglePointData true :bTrackActorsOnlyWithinBounds false :Mode GetSinglePoint))
  (bind SortAttributes_1 (Sort%20Attributes :bUseStableSort false :SortAttributes [(struct :InputSource $Position.Z)]))
  (bind CopyAttribute_8 (Copy%20Attributes :InputSource $Index :OutputTarget SortPrio))
  (bind CreateAttribute_7 (Create%20Constant :AttributeTypes (struct :DoubleValue 1.0)))
  (bind GetPointsCount_5 (Get%20Element%20Count))
  (bind AttributeMathsOp_6 (Attribute%20Maths%20Op :Operation Subtract))
  (bind DuplicatePoint_4 (Duplicate%20Point))
  (bind AttributeFilter_2 (Filter%20Attribute%20Elements :Operator Matches :TargetAttribute ActorReference :ThresholdAttribute ActorReference))
  (bind GetAttributeFromPointIndex_9 (Get%20Attribute%20From%20Point%20Index :bForceOutputAttributeToBeInElementsDomain true :OutputAttributeName SortPrio))
  (bind ApplyOnActor_12 (Apply%20On%20Object :ObjectReferenceAttribute ActorReference :PropertyOverrideDescriptions [(struct :PropertyTarget "SortPrio")]))
  (bind Gather_13 (Gather))
  (bind SplineSampler_1 (Spline%20Sampler :SamplerParams (struct :Mode Distance :DistanceIncrement 2.5e+03 :bFitToCurve true :InteriorSampleSpacing 1e+03 :InteriorBorderSampleSpacing 1e+03 :bUnbounded true)))
  (bind CreateAttribute_6 (Create%20Constant :AttributeTypes (struct :DoubleValue 2e+01)))
  (bind GetAttributeFromPointIndex_4 (Get%20Attribute%20From%20Point%20Index :bForceOutputAttributeToBeInElementsDomain true :InputSource $BoundsMax.X :OutputAttributeName None))
  (bind AttributeMathsOp_5 (Attribute%20Maths%20Op :Operation Divide))
  (bind CreateAttribute_9 (Create%20Constant :AttributeTypes (struct :DoubleValue 1e+01)))
  (bind MakeVectorAttribute_7 (Make%20Vector%20Attribute :OutputType Vector))
  (bind AttributeMathsOp_1 (Attribute%20Maths%20Op :InputSource1 $Scale :Operation Set))
  (bind StaticMeshSpawner_2 (Static%20Mesh%20Spawner :MeshSelectorParameters (class:/Script/PCG.PCGMeshSelectorWeighted :MeshEntries [(struct :Descriptor (struct :StaticMesh /PCG/DebugObjects/PCG_Cube.PCG_Cube :OverrideMaterials [/Game/NeverCook/ArtTools/Tech_Art/BiomeBrushManager/M_BiomeBrush.M_BiomeBrush] :Mobility Movable :bHasCustomNavigableGeometry No :BodyInstance (struct :VelocitySolverIterationCount 1 :CollisionEnabled NoCollision :CollisionProfileName NoCollision :CollisionResponses (struct :ResponseArray [(struct :Channel Visibility :Response ECR_Ignore) (struct :Channel Camera :Response ECR_Ignore)]) :bAutoWeld true) :InstanceStartCullDistance 99999999 :InstanceEndCullDistance 999999999 :bCastShadow false :bAffectDynamicIndirectLighting false :bAffectDistanceFieldLighting false :bCastDynamicShadow false :bCastStaticShadow false :bCastContactShadow false :bReceivesDecals false :bUseGpuLodSelection false :bIncludeInHLOD false))]) :PostProcessFunctionNames [PostSpawn]))
  (connect GetActorData_3 Out ApplyOnActor_12 In)
  (connect GetAttributeFromPointIndex_9 Attribute ApplyOnActor_12 Property%20Overrides)
  (connect DuplicatePoint_4 Out AttributeFilter_2 Filter)
  (connect CopyAttribute_8 Out AttributeFilter_2 In)
  (connect SplineSampler_1 Out AttributeMathsOp_1 InA)
  (connect MakeVectorAttribute_7 Out AttributeMathsOp_1 InB)
  (connect GetAttributeFromPointIndex_4 Attribute AttributeMathsOp_5 InA)
  (connect CreateAttribute_6 Out AttributeMathsOp_5 InB)
  (connect GetPointsCount_5 Out AttributeMathsOp_6 InA)
  (connect CreateAttribute_7 Out AttributeMathsOp_6 InB)
  (connect SortAttributes_1 Out CopyAttribute_8 Source)
  (connect SortAttributes_1 Out CopyAttribute_8 Target)
  (connect GetActorData_3 Out DuplicatePoint_4 In)
  (connect AttributeMathsOp_6 Out DuplicatePoint_4 Iterations)
  (connect ApplyOnActor_12 Out Gather_13 Execution%20Dependency)
  (connect GetSplineData_0 Out Gather_13 In)
  (connect SplineSampler_1 Out GetAttributeFromPointIndex_4 In)
  (connect AttributeFilter_2 InsideFilter GetAttributeFromPointIndex_9 In)
  (connect GetActorData_0 Out GetPointsCount_5 In)
  (connect AttributeMathsOp_5 Out MakeVectorAttribute_7 X)
  (connect AttributeMathsOp_5 Out MakeVectorAttribute_7 Y)
  (connect CreateAttribute_9 Out MakeVectorAttribute_7 Z)
  (connect GetActorData_0 Out SortAttributes_1 In)
  (connect Gather_13 Out SplineSampler_1 Spline)
  (connect AttributeMathsOp_1 Out StaticMeshSpawner_2 In))
