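# Base editor-utility Blueprint for biome brushes, spline actors with instanced static meshes that carry a color channel and texture page index.
# PostSpawn disables Nanite on each instanced mesh, sets translucent sort priority, and passes page index and channel color through custom primitive data.
# The construction script tints the spline by channel.
# Both call UpdateVisuals, which reads EUBP_BiomeBrushManager flags to show or hide meshes and toggle page-index visualization.
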EditorUtilityBlueprint /Game/NeverCook/ArtTools/Tech_Art/BiomeBrushManager/EUBP_BiomeBrush_Base
graphs [
 EventGraph {nodes [
   Event BeginPlay()
   Event ActorBeginOverlap()
   Event Tick!() -> (OutputDelegate, then, DeltaSeconds:float)
  ]}
 Hide {nodes [
   Hide() {is=title K2Node_FunctionEntry}
   > n1 = For Each Loop(Array:Array of Instanced Static Mesh Component Object References=Get Components By Class!(self:Actor Object Reference, ComponentClass:Actor Component Class Reference=/Script/Engine.InstancedStaticMeshComponent) -> (ReturnValue:Array of Instanced Static Mesh Component Object References)) -> (Array Element:Instanced Static Mesh Component Object Reference)
   > SetVisibility!(execute, self:Scene Component Object Reference=@n1, bNewVisibility:bool=false, bPropagateToChildren:bool=false) -> (then)
  ]}
 PostSpawn {nodes [
   PostSpawn() {is=title K2Node_FunctionEntry}
   > n1 = For Each Loop(Array=Get Components By Class(ComponentClass=&InstancedStaticMeshComponent))
   UpdateVisuals(execute=@n1.Completed)
   Is Valid(exec=@n1.LoopBody, InputObject=@n1)
   > Set bDisallowNanite!(execute, bDisallowNanite:bool=true, self:Static Mesh Component Object Reference=@n1) -> (then, Output_Get:bool)
   > SetTranslucentSortPriority!(execute, self:Primitive Component Object Reference=@n1, NewTranslucentSortPriority:int=Get SortPrio() -> (SortPrio:int=0)) -> (then)
   > SetDefaultCustomPrimitiveDataFloat!(
    execute
    self:Primitive Component Object Reference=@n1
    DataIndex:int=0
    Value:float="To Float (Integer)"(InInt=Get TexturePageIndex() -> (TexturePageIndex:int=0))
   ) -> (then)
   > SetVectorParameterForDefaultCustomPrimitiveData!(
    execute
    self:Primitive Component Object Reference=@n1
    ParameterName:Name=Color
    Value:Vector 4 Structure=Select!(
     Option 0:Vector 4 Structure
     Option 1:Vector 4 Structure
     Option 2:Vector 4 Structure
     Option 3:Vector 4 Structure
     Index:Byte=Get ColorChannel() -> (ColorChannel:Byte=0)
    ) -> (ReturnValue:Vector 4 Structure)
   ) -> (then)
   Get Blue() -> (Blue:Vector 4 Structure)
   Get Alpha() -> (Alpha:Vector 4 Structure)
   Get Red() -> (Red:Vector 4 Structure)
   Get Green() -> (Green:Vector 4 Structure)
  ]}
 Show {nodes [
   Show() {is=title K2Node_FunctionEntry}
   > n1 = For Each Loop(Array=Get Components By Class(ComponentClass=&InstancedStaticMeshComponent))
   > SetVisibility(self=@n1, bNewVisibility=true, bPropagateToChildren=false)
  ]}
 UpdatePageIndexVisualization {nodes [
   n1 = UpdatePageIndexVisualization!() -> (then, Visible:bool=~) {is=title K2Node_FunctionEntry}
   > n2 = For Each Loop(Array=Get Components By Class(ComponentClass=&InstancedStaticMeshComponent))
   > SetDefaultCustomPrimitiveDataFloat(self=@n2, DataIndex=5, Value="To Float (Boolean)"(InBool=@n1))
  ]}
 UpdateVisuals {nodes [
   UpdateVisuals() {is=title K2Node_FunctionEntry}
   > n1 = GetActorOfClass!(
    execute
    WorldContextObject:Object Reference
    ActorClass:Actor Class Reference=&/EUBP_BiomeBrushManager.EUBP_BiomeBrushManager_C
   ) -> (then, ReturnValue:EUBP Biome Brush Manager Object Reference)
   > Is Valid(InputObject=@n1)
   > n2 = Branch(Condition=Get VisualizeColorChannel!(self:EUBP Biome Brush Manager Object Reference=@n1) -> (VisualizeColorChannel:bool))
   > n3 = Show() {is=title K2Node_CallFunction}
   UpdatePageIndexVisualization!(
    execute=[@n3.then, @n4.then]
    self
    Visible:bool=Get VisualizePageIndex!(self:EUBP Biome Brush Manager Object Reference=@n1) -> (VisualizePageIndex:bool)
   ) -> (then) {is=title K2Node_CallFunction}
   n4 = Hide(execute=@n2.else) {is=title K2Node_CallFunction}
  ]}
 UserConstructionScript {nodes [
   Construction Script()
   n1 = Select!(Option 0, Option 1, Option 2, Option 3, Index=Get ColorChannel() -> (ColorChannel=0)) -> (ReturnValue_X:double, ReturnValue_Y:double, ReturnValue_Z:double, ReturnValue_W:double)
   > SetUnselectedSplineSegmentColor!(
    execute
    self:Spline Component Object Reference=Get Spline() -> (Spline:Spline Component Object Reference)
    SegmentColor_R:float=@n1
    SegmentColor_G:float=@n1.ReturnValue_Y
    SegmentColor_B:float=@n1.ReturnValue_Z
    SegmentColor_A:float=@n1.ReturnValue_W
   ) -> (then)
   > UpdateVisuals() {is=title K2Node_CallFunction}
   Get Blue()
   Get Alpha()
   Get Red()
   Get Green()
  ]}
]
variables [
 Alpha {type=Vector4, default=<1 1 1 1>}
 Blue {type=Vector4, default=<0 0 1 1>}
 ColorChannel {type=integer}
 Green {type=Vector4, default=<0 1 0 1>}
 Red {type=Vector4, default=<1 0 0 1>}
 SortPrio {type=integer}
 TexturePageIndex {type=integer}
]
