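# Editor utility manager that bakes placed biome brush actors into a biome texture.
# BuildBiomeTexture loops over texture pages, regenerating each page's brush PCG, capturing only those brushes with a top-down scene capture, then restoring editor visibility.
# Each capture becomes a Texture2D fed to a Texture Graph, which exports the composite using OutputTexture's path.
# Construction creates the render target and page textures; helpers refresh brush visualization and run a Python command for pop-out controls.
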EditorUtilityBlueprint /Game/NeverCook/ArtTools/Tech_Art/BiomeBrushManager/EUBP_BiomeBrushManager
graphs [
 BuildBiomeTexture {nodes [
   BuildBiomeTexture() {is=title K2Node_FunctionEntry}
   > Event_BuildBiomeTexture()
  ]}
 Clear {nodes [
   n1 = Clear!() -> (then, GC:bool=~) {is=title K2Node_FunctionEntry}
   > n2 = Sequence()
   > Clear(TargetArray:Array of Texture 2D Object References="Get TexturePages (Read-Only)"() -> ("TexturePages (Read-Only)":Array of Texture 2D Object References)) {is=title K2Node_CallArrayFunction}
   > Set CurrentTexture(CurrentTexture:Texture 2D Object Reference) -> (Output_Get:Texture 2D Object Reference)
   > Clear!(execute, TargetMap:Map of Guid Structures to Booleans=Get ActorVisiblity() -> (ActorVisiblity:Map of Guid Structures to Booleans)) -> (then) {is=title K2Node_CallFunction}
   > Set CurrentIndex(CurrentIndex:int=0) -> (Output_Get:int=0)
   Branch(execute=@n2.then_1, Condition=@n1)
   > CollectGarbage!(execute) -> (then)
  ]}
 EventGraph {nodes [
   Event_BuildBiomeTexture() {is=title K2Node_CustomEvent}
   > n1 = LoadAsset_Blocking(Asset=Get TextureGraph_SR() -> (TextureGraph_SR:Texture Graph Soft Object Reference))
   > n2 = Cast To TextureGraph!(execute, Object=@n1) -> (then, CastFailed, AsTexture Graph:Texture Graph Object Reference) {is=title K2Node_DynamicCast}
   > Set TextureGraph(TextureGraph:Texture Graph Object Reference=@n2) -> (Output_Get:Texture Graph Object Reference)
   > Clear!(execute, self:Self Object Reference, GC:bool=false) -> (then)
   n3 = Get NumPages() -> (NumPages:int=0)
   > Branch(Condition="integer >= integer"(A=int - int(A=@n3)))
   > LoopBody()
   LoopBody() {is=title K2Node_CustomEvent}
   > Clear(TargetArray:Array of EUBP Biome Brush Base Object References=Get RenderActors() -> (RenderActors:Array of EUBP Biome Brush Base Object References)) {is=title K2Node_CallArrayFunction}
   > n4 = Branch(Condition=Get PrintDebug() -> (PrintDebug:bool))
   > n5 = Set BatchActors(BatchActors:String=">>>") -> (Output_Get:String)
   n6 = Set CurrentTexture(
    execute=[@n4.else, @n5.then]
    CurrentTexture="Get (a copy)"(
     Array:Array of Texture 2D Object References=Get TexturePages() -> (TexturePages:Array of Texture 2D Object References)
     Dimension 1=Get CurrentIndex() -> (CurrentIndex:int=0)
    ) -> (Output:Texture 2D Object Reference)
   )
   > Add(TargetArray:Array of Texture 2D Object References="Get TexturePages (Read-Only)"(), NewItem:"Texture 2D Object Reference (by ref)"=@n6) {is=title K2Node_CallArrayFunction}
   > n7 = GetAllActorsOfClass!(
    execute
    WorldContextObject:Object Reference
    ActorClass:Actor Class Reference=&/EUBP_BiomeBrush_Base.EUBP_BiomeBrush_Base_C
   ) -> (then, OutActors:Array of EUBP Biome Brush Base Object References)
   > n8 = Branch(Condition=Get PrintDebug())
   n9 = Get CurrentIndex() -> (CurrentIndex=0)
   n10 = Get CurrentTexture() -> (CurrentTexture:Texture 2D Object Reference)
   > n11 = LogString(InString=Append!(A="Batch - ", B=GetDisplayName(Object=@n10), C:String=": ", D:String="To String (Integer)"(InInt=@n9)) -> (ReturnValue))
   n12 = For Each Loop(Exec=[@n11.then, @n8.else], Array:Array of EUBP Biome Brush Base Object References=@n7) -> (Array Element:EUBP Biome Brush Base Object Reference)
   Set ShowOnlyActors!(
    execute=@n12.Completed
    ShowOnlyActors:Array of Actor Object References=Get RenderActors()
    self:Scene Capture Component Object Reference=Get SceneCaptureComponent2D() -> (SceneCaptureComponent2D:Scene Capture Component 2D Object Reference)
   ) -> (then, Output_Get:Array of Actor Object References)
   > n13 = Branch(Condition=Get PrintDebug())
   > n14 = For Each Loop(Array:Array of Actor Object References=Get ShowOnlyActors!(self:Scene Capture Component Object Reference=Get SceneCaptureComponent2D()) -> (ShowOnlyActors:Array of Actor Object References)) -> (Array Element:Actor Object Reference)
   CaptureScene(execute=[@n13.else, @n14.Completed], self:Scene Capture Component 2D Object Reference=Get SceneCaptureComponent2D())
   n15 = Get ActorVisiblity()
   > n16 = Keys!(execute, TargetMap:Map of Guid Structures to Booleans=@n15) -> (then, Keys:Array of Guid Structures)
   > n17 = Values!(execute, TargetMap:Map of Guid Structures to Booleans=@n15) -> (then, Values:Array of Booleans)
   > n18 = For Each Loop(Array:Array of EUBP Biome Brush Base Object References=Get RenderActors()) -> (Array Element:EUBP Biome Brush Base Object Reference)
   n19 = Get CurrentTexture()
   ConvertRenderTargetToTexture2DEditorOnly(execute=@n18.Completed, RenderTarget=Get RenderTarget() -> (RenderTarget:Texture Render Target 2D Object Reference), Texture=@n19)
   n20 = Append!(A=MakeLiteralString(Value=BiomeTexturePage_), B="To String (Integer)"(InInt=Get CurrentIndex() -> (CurrentIndex=0))) -> (ReturnValue)
   n21 = Get TextureGraph() -> (TextureGraph:Texture Graph Object Reference)
   > SetTextureParameterValue!(
    execute
    WorldContextObject:Object Reference
    TextureGraph:Texture Graph Object Reference=@n21
    ParameterName=String To Name(InString=@n20)
    ParameterValue:Texture Object Reference=@n19
   ) -> (then)
   > SetScalarParameterValue!(
    execute
    WorldContextObject:Object Reference
    TextureGraph:Texture Graph Object Reference=@n21
    ParameterName=BiomeFalloff
    ParameterValue:float="To Float (Integer)"(InInt=Get BiomeFalloff() -> (BiomeFalloff:int=0))
   ) -> (then)
   n22 = Get Object Path String(Object=Get OutputTexture() -> (OutputTexture:Texture 2D Object Reference))
   n23 = Get OutputResolution() -> (OutputResolution:int=0)
   > SetSettingsParameterValue!(
    execute
    WorldContextObject:Object Reference
    TextureGraph:Texture Graph Object Reference=@n21
    ParameterName:Name=Settings
    Width:int=@n23
    Height:int=@n23
    FileName:Name=String To Name(InString=GetBaseFilename(InPath=@n22, bRemovePath=true))
    Path:Name=String To Name(InString=GetPath!(InPath:String=@n22) -> (ReturnValue:String))
    Format:ETG_TextureFormat Enum=BGRA8
    TextureType:ETG_TexturePresetType Enum=None
    LODTextureGroup:TextureGroup Enum=TEXTUREGROUP_World
    Compression:TextureCompressionSettings Enum=TC_Default
    SRGB:bool=false
   ) -> (then)
   > n24 = Branch(Condition=Get PrintDebug())
   > n25 = LogString(InString=Append!(A=GetDisplayName(Object=@n21), B=" - ", C:String=GetDisplayName(Object=@n19), D:String=" - ", E:String=@n20) -> (ReturnValue))
   Increment Int(" "=[@n25.then, @n24.else], Value=Get CurrentIndex() -> (CurrentIndex=0))
   > n26 = Branch(Condition="integer < integer"(A=Get CurrentIndex() -> (CurrentIndex=0), B=Get NumPages() -> (NumPages=0)))
   > LoopBody() {is=title K2Node_CallFunction}
   CompleteLoop(execute=@n26.else)
   n27 = Branch(execute=@n18.LoopBody, Condition=Get VisualizeColorChannel() -> (VisualizeColorChannel:bool))
   > n28 = Show(self:EUBP Biome Brush Base Object Reference=@n18)
   n29 = Find Item!(
    TargetArray:Array of Guid Structures=@n16
    ItemToFind:"Guid Structure (by ref)"=Get ActorGuid!(self:Actor Object Reference=@n18) -> (ActorGuid:Guid Structure)
   ) -> (ReturnValue)
   Branch(execute=[@n28.then, @n30.then], Condition="Not Equal (Integer)"(A=@n29, B=-1))
   > SetIsTemporarilyHiddenInEditor!(
    execute
    self:Actor Object Reference=@n18
    bIsHidden:bool="Get (a ref)"(Array:Array of Booleans=@n17, Dimension 1=@n29) -> (Output:"Boolean (by ref)")
   ) -> (then)
   n30 = Hide(execute=@n27.else, self:EUBP Biome Brush Base Object Reference=@n18)
   LogString(execute=@n14.LoopBody, InString=Append!(A=">>>", B=GetDisplayName(Object=@n14)) -> (ReturnValue))
   Branch(
    execute=@n12.LoopBody
    Condition="Equal (Integer)"(
     A=Get TexturePageIndex!(self:EUBP Biome Brush Base Object Reference=@n12) -> (TexturePageIndex:int=0)
     B=Get CurrentIndex() -> (CurrentIndex=0)
    )
   )
   > Add!(
    execute
    TargetMap:Map of Guid Structures to Booleans=Get ActorVisiblity()
    Key:"Guid Structure (by ref)"=Get ActorGuid(self=@n12)
    Value:"Boolean (by ref)"=IsHiddenEd!(self:Actor Object Reference=@n12) -> (ReturnValue:bool)
   ) -> (then) {is=title K2Node_CallFunction}
   > SetIsTemporarilyHiddenInEditor(self=@n12, bIsHidden=false)
   > Show(self=@n12)
   > Add(
    TargetArray:Array of EUBP Biome Brush Base Object References=Get RenderActors()
    NewItem:"EUBP Biome Brush Base Object Reference (by ref)"=@n12
   ) {is=title K2Node_CallArrayFunction}
   > n31 = Branch(Condition=Get PrintDebug())
   > n32 = Set BatchActors(BatchActors=Append!(A=Get BatchActors() -> (BatchActors:String), B=GetDisplayName(Object=@n12), C:String=">>>") -> (ReturnValue))
   GenerateLocal!(
    execute=[@n31.else, @n32.then]
    self:PCGComponent Object Reference=Get PCG!(self:EUBP Biome Brush Base Object Reference=@n12) -> (PCG:PCGComponent Object Reference)
    bForce:bool=true
   ) -> (then)
   CompleteLoop() {is=title K2Node_CustomEvent}
   > n33 = Branch(Condition=Get PrintDebug())
   > n34 = GetTextureParameterValue!(
    execute
    WorldContextObject:Object Reference
    TextureGraph:Texture Graph Object Reference=Get TextureGraph()
    ParameterName:Name=BiomeTexturePage_0
   ) -> (then, ReturnValue:Texture Object Reference)
   > n35 = GetTextureParameterValue(TextureGraph=Get TextureGraph(), ParameterName=BiomeTexturePage_1)
   > n36 = GetTextureParameterValue(TextureGraph=Get TextureGraph(), ParameterName=BiomeTexturePage_2)
   > n37 = GetTextureParameterValue(TextureGraph=Get TextureGraph(), ParameterName=BiomeTexturePage_3)
   > n38 = LogString(InString=Append!(
     A="To String (Object)"!(InObj:Object Reference=@n34) -> (ReturnValue:String)
     B=|
     C="To String (Object)"(InObj=@n35)
     D:String=|
     E:String="To String (Object)"(InObj=@n36)
     F:String=|
     G:String="To String (Object)"(InObj=@n37)
    ) -> (ReturnValue))
   n39 = "Get CheckoutAndSave (slow)"() -> ("CheckoutAndSave (slow)":bool)
   n40 = Get TextureGraph()
   ExportTextureGraph!(
    execute=[@n38.then, @n33.else]
    WorldContextObject:Object Reference
    InTextureGraph:Texture Graph Base Object Reference=@n40
    bOverwriteTextures:bool=true
    bSave:bool=@n39
    bExportAll:bool=false
   ) -> (then)
   "Texture Graph Export (Async)"!(
    execute:Exec
    TextureGraph:Texture Graph Base Object Reference=@n40
    OverwriteTextures:bool=true
    bSave:bool=@n39
    bExportAll:bool=false
    bDisableCache:bool=false
   ) -> (then:Exec, OnDone:Exec) {is=title K2Node_AsyncAction}
  ]}
 Pop Out Controls {nodes [
   Pop Out Controls() {is=title K2Node_FunctionEntry}
   > ExecutePythonCommand(PythonCommand=Replace(SourceString=Get Command() -> (Command:String), From="##", To="\r\n"))
  ]}
 RefreshVisualziation {nodes [
   RefreshVisualziation() {is=title K2Node_FunctionEntry}
   > n1 = GetAllActorsOfClass(ActorClass=&EUBP_BiomeBrush_Base_C)
   > n2 = For Each Loop(Array=@n1)
   > GenerateLocal(self=Get PCG(self=@n2), bForce=true)
  ]}
 UserConstructionScript {nodes [
   Construction Script()
   > n1 = Sequence!(execute) -> (then_0, then_1, then_2:Exec)
   > Set World Transform!(
    execute
    self:Scene Component Object Reference=Get SceneCaptureComponent2D()
    NewTransform:"Transform (by ref)"=MakeTransform!(Location_X:double=0, Location_Y:double=0, Location_Z:double=100000, Rotation="-90,-90,0", Scale) -> (ReturnValue)
    bSweep:bool=false
    bTeleport:bool=false
   ) -> (then, SweepHitResult:Hit Result Structure) {is=title K2Node_CallFunction}
   n2 = Is Valid(exec=@n1.then_1, InputObject=Get RenderTarget())
   n3 = Get NumPages() -> (NumPages=0)
   n4 = "Max (Integer)"(A=int / int(A=Get OutputResolution() -> (OutputResolution=0), B=int / int(A=@n3, B=2)), B=32)
   n5 = CreateRenderTarget2D!(
    execute=@n2.Is Not Valid
    WorldContextObject:Object Reference
    Width:int=@n4
    Height:int=@n4
    Format:ETextureRenderTargetFormat Enum=RTF_RGBA8
    ClearColor:Linear Color Structure=(R=0,G=0,B=0,A=1)
    bAutoGenerateMipMaps:bool=false
    bSupportUAVs:bool=false
   ) -> (then, ReturnValue:Texture Render Target 2D Object Reference)
   > n6 = Set RenderTarget(RenderTarget:Texture Render Target 2D Object Reference=@n5) -> (Output_Get:Texture Render Target 2D Object Reference)
   > Set TextureTarget!(
    execute
    TextureTarget:Texture Render Target 2D Object Reference=@n6
    self:Scene Capture Component 2D Object Reference=Get SceneCaptureComponent2D()
   ) -> (then, Output_Get:Texture Render Target 2D Object Reference)
   n7 = Get TexturePages()
   Branch(
    execute=@n1.then_2
    Condition="Not Equal (Integer)"(A=Length(TargetArray:Array of Texture 2D Object References=@n7), B=Get NumPages() -> (NumPages=0))
   )
   > Resize!(execute, TargetArray:Array of Texture 2D Object References=@n7, Size:int=Get NumPages() -> (NumPages=0)) -> (then) {is=title K2Node_CallArrayFunction}
   > n8 = For Loop(LastIndex=Last Index!(TargetArray:Array of Texture 2D Object References=@n7) -> (ReturnValue=0))
   > n9 = Construct Object from Class(Class=/Script/Engine.Texture2D) -> (ReturnValue:Texture 2D Object Reference)
   > Set DeferCompression!(execute, DeferCompression:bool=true, self:Texture Object Reference=@n9) -> (then, Output_Get:bool)
   > Set CompressionSettings(CompressionSettings=TC_BC7, self=@n9)
   > Set Array Elem!(
    execute
    TargetArray:Array of Texture 2D Object References=Get TexturePages()
    Index:int=@n8
    Item:"Texture 2D Object Reference (by ref)"=@n9
    bSizeToFit:bool=false
   ) -> (then)
  ]}
]
variables [
 ActorVisiblity {}
 BatchActors {type=string, default=""}
 BiomeCellSize {type=integer, default=2000}
 BiomeFalloff {type=integer, default=1}
 "CheckoutAndSave (slow)" {type=boolean}
 CurrentIndex {type=integer}
 CurrentTexture {type=Texture2D}
 NumPages {type=integer, default=4}
 OutputResolution {type=integer, default=256}
 OutputTexture {type=Texture2D}
 PrintDebug {type=boolean}
 RenderActors {type="array<EUBP_BiomeBrush_Base_C>", default []}
 RenderTarget {type=TextureRenderTarget2D, default=null}
 TextureGraph {type=TextureGraph, default=null}
 TextureGraph_SR {type=TextureGraph, default=&/TG_BiomeBrushManager_CompositePages}
 TexturePages {type="array<Texture2D>", default []}
 "TexturePages (Read-Only)" {type="array<Texture2D>", default []}
 VisualizeColorChannel {type=boolean}
 VisualizePageIndex {type=boolean}
]
